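; PCG subgraph that runs a Mesh Partition Query for mesh terrain inside a box sized by a parameter.
; Divide and multiply ops derive max and min bounds from that size for a created point, which replaces the PCG volume bounds as bounding shape.
; Parameters drive the target partition, query type, modifier layer and sub-priority, inclusivity and channels.
; The query recomputes vertex normals and feeds the graph output.
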(graph Get_MeshTerrain_CoreProcess
  (params (MeshPartitionBoundSize :type Vector :default (1e+05 1e+05 1e+05)) (MeshPartitionTarget :type (SoftObject /Script/MeshPartition.MeshPartition)) (MeshPartitionQueryType :type (Enum /Script/PCGMeshPartitionInterop.EPCGQueryType)) (ModifierLayerPriority :type Name) (ModifierSubPriority :type Double) (MeshPartitionQueryInclusive :type Bool) (Channels :type (Array Name) :default [None]))
  (bind MegaMeshQuery_1 (Mesh%20Partition%20Query :QueryParams (struct :QueryType Intermediate :bRecomputeVertexNormals true) :bInclusive MeshPartitionQueryInclusive :Channels Channels :LayerName ModifierLayerPriority :MegaMeshOverride MeshPartitionTarget :QueryType MeshPartitionQueryType :SubPriority ModifierSubPriority))
  ;[0 0 0] Replacing PCG volume bound size
  (bind AttributeMathsOp_1 (Attribute%20Maths%20Op :Operation Divide :InA MeshPartitionBoundSize))
  (bind AttributeMathsOp_2 (Attribute%20Maths%20Op :Operation Multiply))
  (bind CreatePoints_0 (Create%20Points))
  (connect AttributeMathsOp_1 Out AttributeMathsOp_2 InA)
  (connect AttributeMathsOp_1 Out CreatePoints_0 BoundsMax)
  (connect AttributeMathsOp_2 Out CreatePoints_0 BoundsMin)
  (connect CreatePoints_0 Out MegaMeshQuery_1 Bounding%20Shape)
  (output Out MegaMeshQuery_1))
